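# Unlit translucent two-sided trail material for the ScallopLava pickaxe, built for Niagara ribbons and mesh particles.
# Emissive combines curve-atlas color with panning noise, smoke and crack layers for a lava look, tinted by particle color.
# Opacity combines a scrolling trail mask, noise erosion, HeightLerp dissolve and depth fade, with dynamic parameters driving dissolve, offset, width and appearance.
# World position offset optionally applies SplineThicken plus a camera-transformed sine wobble.
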Material /BRCosmetics/Weapons/FORT_Melee/Pickaxe_ScallopLava/FX/M_Pickaxe_ScallopLava_TrailMain
graph {
 properties {
  bAutomaticallySetUsageInEditor=false
  bUsedWithNiagaraMeshParticles=true
  bUsedWithNiagaraRibbons=true
  blendMode=BLEND_Translucent
  shadingModel=MSM_Unlit
  twoSided=true
  usedShadingModels=Unlit
 }
 nodes [
  n1 = TextureCoordinate() {coordinateIndex=1}
  n2 = Constant2Vector() {r=0.5}
  n3 = ScalarParameter() {defaultValue=1, parameterName=PannerSpeed}
  n4 = TextureSampleParameter2D(UVs=Panner(Coordinate=TextureCoordinate() {uTiling=1.3}, Speed=Multiply(A=@n2, B=@n3)) {speedX=0.5}) {parameterName=Tex_Default, texture=&/T_ScallopLava_Trail}
  n5 = OneMinus(None=ComponentMask(None=TextureCoordinate()) {r=true})
  n6 = Multiply(
   A=Saturate(None=StaticComponentMaskParameter(None=@n4.RGBA) {defaultR=true, parameterName=Mask})
   B=Saturate(None=Multiply(A=@n5) {constB=5})
  )
  n7 = Constant() {r=-3}
  n8 = Constant() {r=2}
  n9 = Add(
   A=TextureCoordinate() {uTiling=3}
   B=AppendVector(A=DynamicParameter() {defaultValue=<rgba 1 0 -1 1>, paramNames [Use dissolve, U_Offset, SplineWidth, Appear]}.U_Offset, B=Constant())
  ) {constB=0}
  n10 = Multiply(
   A=ComponentMask(None=VectorParameter() {parameterName=DissolveScrollSpeed, defaultValue=<rgba 1 0 0 1>}.RGB) {g=true, r=true}
   B=ScalarParameter() {defaultValue=1, parameterName=PannerSpeed}
  )
  n11 = Constant()
  n12 = Constant() {r=0.5}
  n13 = Power(Base=ComponentMask(None=TextureCoordinate()) {r=true}) {constExponent=1.3}
  n14 = Power(
   Base=TextureSample(UVs=Panner(Coordinate=@n9, Speed=@n10) {speedX=-0.5}) {texture=/Game/Effects/Niagara/MasterMaterials/SplineThickenTrail/Texture/T_Pickaxe_SoapPocket_NoiseGlow}.R
   Exp=Saturate(None=LinearInterpolate(A=@n11, B=@n12, Alpha=@n13) {constA=0.1, constB=2})
  )
  n15 = LinearInterpolate(A=Saturate(None=LinearInterpolate(A=@n7, B=@n8, Alpha=@n14) {constA=-2, constB=2}), Alpha=ScalarParameter() {parameterName=UseSimpleGlow})
  n16 = ComponentMask(None=TextureCoordinate()) {r=true}
  n17 = Multiply(A=Multiply(A=Multiply(A=@n6, B=@n15)), B=Saturate(None=Multiply(A=@n16) {constB=10}))
  n18 = ScalarParameter() {defaultValue=10, parameterName=FadeDistance}
  n19 = DynamicParameter() {defaultValue=<rgba 1 0 -1 1>, paramNames [Use dissolve, U_Offset, SplineWidth, Appear]}
  n20 = VectorParameter() {parameterName=DissolveUVScale, defaultValue=<rgba 3 0.7 0 0>}
  n21 = VectorParameter() {parameterName=DissolvePanner, defaultValue=<rgba 0.2 0 0 0>}
  n22 = Panner(
   Coordinate=Multiply(A=TextureCoordinate(), B=ComponentMask(None=@n20.RGB) {g=true, r=true})
   Speed=Multiply(A=ComponentMask(None=@n21.RGB) {g=true, r=true}, B=ScalarParameter() {defaultValue=1, parameterName=PannerSpeed})
  ) {speedX=0.2}
  n23 = Multiply(A=MaterialFunctionCall(
    Transition Phase=LinearInterpolate(Alpha=@n19.Use dissolve) {constA=0.4, constB=0.8}
    Height Texture=TextureSample(UVs=@n22) {texture=&T_Pickaxe_SoapPocket_NoiseGlow}.R
    Contrast=Constant() {r=2}
   ) {materialFunction=&HeightLerp}.Alpha)
  n24 = Multiply(
   A=Multiply(A=DepthFade(Opacity=@n17, FadeDistance=@n18), B=LinearInterpolate(Alpha=@n23) {constA=-2, constB=2})
   B=Sine(None=ComponentMask(None=@n1) {r=true}) {period=2}
  )
  n25 = Add(
   A=ComponentMask(None=TextureCoordinate()) {r=true}
   B=LinearInterpolate(Alpha=DynamicParameter() {defaultValue=<rgba 1 0 -1 1>, paramNames [Use dissolve, U_Offset, SplineWidth, Appear]}.Appear) {constA=-1}
  )
  n26 = LinearInterpolate(B=Saturate(None=Multiply(A=@n25) {constB=3}), Alpha=ScalarParameter() {parameterName=UseAppear}) {constA=1}
  DepthFade(Opacity=Multiply(A=Multiply(A=@n24, B=@n26), B=ParticleColor().A)) {fadeDistanceDefault=30}
  n27 = DynamicParameter() {defaultValue=<rgba 1 0 -1 1>, paramNames [Use dissolve, U_Offset, SplineWidth, Appear]}
  n28 = Power(Base=ComponentMask(None=TextureCoordinate()) {r=true}, Exp=ScalarParameter() {defaultValue=1, parameterName=Top WPO Fade})
  n29 = TextureCoordinate()
  n30 = Power(Base=OneMinus(None=ComponentMask(None=@n29) {r=true})) {constExponent=3}
  n31 = LinearInterpolate(
   A=Saturate(None=Multiply(A=@n28) {constB=3})
   B=LinearInterpolate(Alpha=Multiply(A=@n30) {constB=3}) {constA=1, constB=1.5}
   Alpha=ScalarParameter() {parameterName=UseWidthMaskInvert}
  )
  n32 = Multiply(A=Multiply(A=ScalarParameter() {defaultValue=20, parameterName=WidthBase}, B=@n27.SplineWidth), B=@n31)
  n33 = Multiply(A=Multiply(A=ScalarParameter() {defaultValue=20, parameterName=WidthTip}, B=@n27.SplineWidth), B=@n31)
  n34 = StaticBool()
  n35 = Time()
  n36 = Add(
   A=ComponentMask(None=TextureCoordinate()) {r=true}
   B=Multiply(A=Multiply(A=@n35) {constB=-0.5}, B=ScalarParameter() {defaultValue=1, parameterName=PannerSpeed})
  )
  n37 = VectorParameter() {parameterName=WPO Direction, defaultValue=<rgba 0 0 17 0>}
  n38 = Multiply(A=Multiply(A=Sine(None=@n36) {period=0.8}, B=Transform(None=@n37.RGB) {transformType=TRANSFORM_Camera}))
  n39 = ComponentMask(None=TextureCoordinate()) {r=true}
  n40 = Saturate(None=Multiply(A=Power(Base=@n39)))
  LinearInterpolate(
   B=Add(
    A=MaterialFunctionCall(WidthBase=@n32, WidthTip=@n33, Expand U or V UV Channel=@n34) {materialFunction=&SplineThicken}.WorldPositionOffset
    B=Multiply(A=@n38, B=@n40)
   )
   Alpha=ScalarParameter() {parameterName=UseSplineThicken}
  )
  n41 = ScalarParameter() {parameterName=UseDes}
  n42 = CurveAtlasRowParameter(CurveTime=ComponentMask(None=TextureCoordinate()) {r=true}) {
   atlas=&/CAtlas_ScallopLava_FXCurve
   curve=&/CV_ScallopLava_FXTrailColor01
   parameterName=Color
  }
  n43 = TextureCoordinate() {uTiling=1.3}
  n44 = Multiply(A=Constant2Vector() {r=0.5}, B=ScalarParameter() {defaultValue=1, parameterName=PannerSpeed})
  n45 = StaticComponentMaskParameter(None=TextureSampleParameter2D(UVs=Panner(Coordinate=@n43, Speed=@n44) {speedX=0.5}) {parameterName=Tex_Default, texture=&T_ScallopLava_Trail}.RGBA) {defaultR=true, parameterName=Mask}
  n46 = Power(Base=@n45) {constExponent=10}
  n47 = Panner(
   Coordinate=TextureCoordinate() {uTiling=5}
   Speed=Multiply(A=Constant2Vector() {r=0.5}, B=ScalarParameter() {defaultValue=1, parameterName=PannerSpeed})
  ) {speedX=0.5}
  n48 = VectorParameter() {parameterName=DarkColor, defaultValue=<rgba 0.1 0.0185 0.0061 1>}
  n49 = VectorParameter() {parameterName=DetailColor, defaultValue=<rgba 5 2.4029 0.3197 1>}
  n50 = TextureCoordinate() {uTiling=5, vTiling=1.5}
  n51 = Multiply(A=Constant2Vector() {r=1.2}, B=ScalarParameter() {defaultValue=1, parameterName=PannerSpeed})
  n52 = Power(Base=TextureSample(UVs=Panner(Coordinate=@n50, Speed=@n51) {speedX=1.2}) {texture=&T_Inky_Smoke_Tile}.R) {constExponent=1.5}
  n53 = Add(
   A=Multiply(
    A=TextureSample(UVs=@n47) {samplerType=SAMPLERTYPE_Grayscale, texture=/Game/Global/Textures/Default/Noise/TilingNoise06}.R
    B=VectorParameter() {parameterName=CoreColor, defaultValue=<rgba 10 0.232 0 1>}.RGB
   )
   B=Multiply(A=LinearInterpolate(A=@n48.RGB, B=@n49.RGB, Alpha=@n52))
  )
  n54 = Constant2Vector() {r=0.5}
  n55 = ScalarParameter() {defaultValue=1, parameterName=PannerSpeed}
  n56 = TextureSample(UVs=Panner(Coordinate=TextureCoordinate() {uTiling=3}, Speed=Multiply(A=@n54, B=@n55)) {speedX=0.5}) {samplerType=SAMPLERTYPE_LinearColor, texture=/Game/Global/Textures/Default/Masks_and_Gradients/T_Cracks}
  n57 = ScalarParameter() {defaultValue=2, parameterName=HighLightPower}
  n58 = Multiply(
   A=Power(Base=@n45) {constExponent=5}
   B=Multiply(A=VectorParameter() {parameterName=CoreDetailColor, defaultValue=<rgba 100 16.4838 0 1>}.RGB, B=Power(Base=@n56.R, Exp=@n57))
  )
  n59 = Add(A=LinearInterpolate(A=LinearInterpolate(Alpha=@n46) {constA=0.5}, B=Add(A=@n53, B=@n58) {constB=0}, Alpha=Power(Base=@n45) {constExponent=3})) {constB=0}
  n60 = ScalarParameter() {parameterName=UseSimpleGlow}
  n61 = Add(A=Multiply(A=LinearInterpolate(A=@n42, Alpha=@n41), B=LinearInterpolate(A=@n59, Alpha=@n60) {constA=1}) {constA=1}) {constB=0}
  n62 = TextureCoordinate()
  n63 = Power(Base=OneMinus(None=ComponentMask(None=@n62) {r=true})) {constExponent=3}
  n64 = LinearInterpolate(B=Saturate(None=Multiply(A=@n63) {constB=8}), Alpha=ScalarParameter() {parameterName=UseTailDarkGradiant}) {constA=1, constB=0}
  MaterialFunctionCall(Emissive=Multiply(A=ParticleColor().RGB, B=Multiply(A=@n61, B=@n64))) {materialFunction=&MF_FX_EmissiveAdjustments}
 ]
}
